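# Niagara system for the Vertigo large male skin, attaching circles, mesh tendrils and ribbons to the character's skeletal mesh.
# Four emitters are live: partitioned BoneFlow circles and rotating ribbons, socket-sampled tendrils, and curl-noise trail ribbons; five variants are disabled.
# Trail spawn rate and width follow a system integrator, spawn rate drops to zero while skydiving, and lifetime shortens with owner speed.
# Both ribbon emitters use distance culling.
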NiagaraSystem /BRCosmetics/Characters/Player/Male/Large/Bodies/M_LRG_Vertigo/FX/NS_SKin_Vertigo
dependencies {
 usedDataInterfaces [&NiagaraDataInterfaceCurve, &NiagaraDataInterfaceVectorCurve]
 usedDynamicInputs []
 usedModules [
  &self:SystemSpawnScript.SystemScriptSource.SystemScriptGraph.NiagaraNodeAssignment_0.SetVariables_#1
  &FN_Integrator
  &SystemState
 ]
}
emitters {
 Circles_Bone_Flow {
  bEnabled=false
  emitterSpawnScript [
   /Game/Effects/Niagara/Modules/Partition/FN_PartitionAttributes(Number of Bone Sets=2, Particles Per Partition=4, Partitions=2)
   /SetVariables(Emitter.Seed:NiagaraInt32=/Niagara/DynamicInputs/Random/RandomInt(Randomness Mode:ENiagaraRandomnessMode=::NewEnumerator0, Random Seed:NiagaraInt32=0, Fixed Random Seed:NiagaraBool=false)) {isSetParameters=true, module=SetVariables_#2}
  ]
  emitterUpdateScript [
   &EmitterState()
   &SpawnBurst_Instantaneous(Spawn Count=$Emitter.FN_PartitionAttributes.Spawn Total, Use Spawn Probability=false) {enabled=false}
   &SpawnRate(SpawnRate=3, Use Spawn Probability=false)
  ]
  particleSpawnScript [
   /Game/Effects/Niagara/Modules/Partition/FN_Particle_partition_v5/FN_Particle_Partition_v5(Enable Ribbon ID=true, Pattern:ENiagaraPartitionWaveForm=::NewEnumerator1, Seed:NiagaraInt32=$Emitter.Seed)
   /Niagara/Modules/Spawn/Initialization/InitializeParticle(
    Lifetime=1.5
    Mesh Scale=&Multiply_VectorByFloat(Vector=<1 0.8 1>, Float=&RandomRangeFloat(Minimum=0.7, Maximum=1))
    Module.Write Color=true
    Module.Write Lifetime=true
    Module.Write Mass=true
    Module.Write Material Random=true
    Module.Write Position=false
    Module.Write Scale=true
    Module.Write SpriteRotation=false
    Module.Write SpriteSize=true
    Position=&SimulationPosition()
    Sprite Size=<25 25>
   )
   /Game/Effects/Niagara/Modules/BoneFlow/FN_Initialize_BoneFlow_Attributes_v2()
  ]
  particleUpdateScript [
   &ParticleState()
   &FN_BoneFlow(
    Angle:NiagaraFloat=$Particles.FN_Initialize_BoneFlow_Attributes_v2.Angle
    Height:NiagaraFloat=$Particles.Partition.NormalizedID
    Radius:NiagaraFloat=$Particles.Initial.FN_Initialize_BoneFlow_Attributes_v2.Radius
    SkeletalMesh:NiagaraDataInterfaceSkeletalMesh=$User.SkeletalMesh
    Use Sockets Instead=true
    Write Position Attribute=true
   )
   &ScaleMeshSize(Scale Factor=&VectorFromFloat(Value=&FloatFromCurve(FloatCurve=DataInterface() {propertyValues {Curve {keys [
          {arriveTangent=3.7291, interpMode=RCIM_Cubic, leaveTangent=3.7291, tangentMode=RCTM_User, time=0, value=0}
          <at 0.5 1 RCIM_Cubic>
          {arriveTangent=-2.7655, interpMode=RCIM_Cubic, leaveTangent=-2.7655, tangentMode=RCTM_User, time=1, value=0}
         ]}}})))
   &DynamicMaterialParameters(
    Index 0 Param 1=&FloatFromCurve(FloatCurve=DataInterface() {propertyValues {Curve {keys [<at 0 0 RCIM_Cubic>,<at 0.5 1 RCIM_Cubic>,<at 1 0 RCIM_Cubic>]}}}, Scale Curve=1.25)
    Index 0 Param 2=&FloatFromCurve(
     FloatCurve=DataInterface() {propertyValues {Curve {keys [<at 0 1 RCIM_Cubic>,<at 0.5 0.1 RCIM_Cubic>,<at 1 1 RCIM_Cubic>]}}}
     Scale Curve=3.5
    )
    Index 0 Param 4=1
    Module.Param0WriteEnabled=true
    Module.Param1WriteEnabled=true
    Module.Param2WriteEnabled=true
    Module.Param3WriteEnabled=true
   )
  ]
  rendererClasses [&NiagaraSpriteRendererProperties, &NiagaraMeshRendererProperties]
  renderers [
   {rendererClass=&NiagaraSpriteRendererProperties}
   {rendererClass=&NiagaraMeshRendererProperties, rendererIndex=1}
  ]
 }
 Circles_Bone_Flow001 {
  bEnabled=false
  emitterSpawnScript [
   &FN_PartitionAttributes(Number of Bone Sets=2, Particles Per Partition=4, Partitions=2)
   /SetVariables(Emitter.Seed:NiagaraInt32=&RandomInt(Randomness Mode:ENiagaraRandomnessMode=::NewEnumerator0, Random Seed:NiagaraInt32=0, Fixed Random Seed:NiagaraBool=false)) {isSetParameters=true, module=SetVariables_#2}
  ]
  emitterUpdateScript [
   &EmitterState()
   &SpawnBurst_Instantaneous(Spawn Count=$Emitter.FN_PartitionAttributes.Spawn Total, Use Spawn Probability=false) {enabled=false}
   &SpawnRate(SpawnRate=6, Use Spawn Probability=false)
  ]
  particleSpawnScript [
   &FN_Particle_Partition_v5(Enable Ribbon ID=true, Pattern:ENiagaraPartitionWaveForm=::NewEnumerator1, Seed:NiagaraInt32=$Emitter.Seed)
   &InitializeParticle(
    Color=<rgba 0.276 0.276 0.276 1>
    Lifetime=1.5
    Mesh Scale=&Multiply_VectorByFloat(Vector=<1 0.8 1>, Float=&RandomRangeFloat(Minimum=0.7, Maximum=1))
    Module.Write Color=true
    Module.Write Lifetime=true
    Module.Write Mass=true
    Module.Write Material Random=true
    Module.Write Position=false
    Module.Write Scale=true
    Module.Write SpriteRotation=false
    Module.Write SpriteSize=true
    Position=&SimulationPosition()
    Sprite Size=<25 25>
   )
   &FN_Initialize_BoneFlow_Attributes_v2()
  ]
  particleUpdateScript [
   &ParticleState()
   &FN_BoneFlow(
    Angle:NiagaraFloat=$Particles.FN_Initialize_BoneFlow_Attributes_v2.Angle
    Height:NiagaraFloat=$Particles.Partition.NormalizedID
    Radius:NiagaraFloat=$Particles.Initial.FN_Initialize_BoneFlow_Attributes_v2.Radius
    SkeletalMesh:NiagaraDataInterfaceSkeletalMesh=$User.SkeletalMesh
    Use Sockets Instead=true
    Write Position Attribute=true
   )
   &ScaleMeshSize(Scale Factor=&VectorFromFloat(Value=&FloatFromCurve(FloatCurve=DataInterface() {propertyValues {Curve {keys [
          {arriveTangent=3.7291, interpMode=RCIM_Cubic, leaveTangent=3.7291, tangentMode=RCTM_User, time=0, value=0}
          <at 0.5 1 RCIM_Cubic>
          {arriveTangent=-2.7655, interpMode=RCIM_Cubic, leaveTangent=-2.7655, tangentMode=RCTM_User, time=1, value=0}
         ]}}})))
   &DynamicMaterialParameters(
    Index 0 Param 1=&FloatFromCurve(FloatCurve=DataInterface() {propertyValues {Curve {keys [<at 0 0 RCIM_Cubic>,<at 0.5 1 RCIM_Cubic>,<at 1 0 RCIM_Cubic>]}}}, Scale Curve=1.25)
    Index 0 Param 2=&FloatFromCurve(
     FloatCurve=DataInterface() {propertyValues {Curve {keys [<at 0 1 RCIM_Cubic>,<at 0.5 0.1 RCIM_Cubic>,<at 1 1 RCIM_Cubic>]}}}
     Scale Curve=3.5
    )
    Index 0 Param 4=1
    Module.Param0WriteEnabled=true
    Module.Param1WriteEnabled=true
    Module.Param2WriteEnabled=true
    Module.Param3WriteEnabled=true
   )
  ]
  rendererClasses [&NiagaraSpriteRendererProperties, &NiagaraMeshRendererProperties]
  renderers [
   {rendererClass=&NiagaraSpriteRendererProperties}
   {rendererClass=&NiagaraMeshRendererProperties, rendererIndex=1}
  ]
 }
 Circles_Bone_Flow002 {
  emitterSpawnScript [
   &FN_PartitionAttributes(Number of Bone Sets=2, Particles Per Partition=4, Partitions=2)
   /SetVariables(Emitter.Seed:NiagaraInt32=&RandomInt(Randomness Mode:ENiagaraRandomnessMode=::NewEnumerator0, Random Seed:NiagaraInt32=0, Fixed Random Seed:NiagaraBool=false)) {isSetParameters=true, module=SetVariables_#2}
  ]
  emitterUpdateScript [
   &EmitterState()
   &SpawnBurst_Instantaneous(Spawn Count=$Emitter.FN_PartitionAttributes.Spawn Total, Use Spawn Probability=false) {enabled=false}
   &SpawnRate(SpawnRate=6, Use Spawn Probability=false)
  ]
  particleSpawnScript [
   &FN_Particle_Partition_v5(Enable Ribbon ID=true, Pattern:ENiagaraPartitionWaveForm=::NewEnumerator1, Seed:NiagaraInt32=$Emitter.Seed) {enabled=false}
   &self:ParticlePartition_SP(Enable Ribbon ID=true, Seed:NiagaraInt32=$Emitter.Seed) {module=ParticlePartition_SP}
   &InitializeParticle(
    Lifetime=1.5
    Mesh Scale=&Multiply_VectorByFloat(Vector=<1 0.8 1>, Float=&RandomRangeFloat(Minimum=0.7, Maximum=1))
    Module.Write Color=true
    Module.Write Lifetime=true
    Module.Write Mass=true
    Module.Write Material Random=true
    Module.Write Position=false
    Module.Write Scale=true
    Module.Write SpriteRotation=false
    Module.Write SpriteSize=true
    Position=&SimulationPosition()
    Sprite Size=<25 25>
   )
   &FN_Initialize_BoneFlow_Attributes_v2()
  ]
  particleUpdateScript [
   &ParticleState()
   &FN_BoneFlow(
    Angle:NiagaraFloat=$Particles.FN_Initialize_BoneFlow_Attributes_v2.Angle
    Height:NiagaraFloat=$Particles.Partition.NormalizedID
    Radius:NiagaraFloat=$Particles.Initial.FN_Initialize_BoneFlow_Attributes_v2.Radius
    SkeletalMesh:NiagaraDataInterfaceSkeletalMesh=$User.SkeletalMesh
    Use Sockets Instead=true
    Write Position Attribute=true
   )
   &ScaleMeshSize(Scale Factor=&VectorFromFloat(Value=&FloatFromCurve(FloatCurve=DataInterface() {propertyValues {Curve {keys [
          {arriveTangent=3.7291, interpMode=RCIM_Cubic, leaveTangent=3.7291, tangentMode=RCTM_User, time=0, value=0}
          <at 0.5 1 RCIM_Cubic>
          {arriveTangent=-2.7655, interpMode=RCIM_Cubic, leaveTangent=-2.7655, tangentMode=RCTM_User, time=1, value=0}
         ]}}})))
   &DynamicMaterialParameters(
    Index 0 Param 1=&FloatFromCurve(FloatCurve=DataInterface() {propertyValues {Curve {keys [<at 0 0 RCIM_Cubic>,<at 0.5 1 RCIM_Cubic>,<at 1 0 RCIM_Cubic>]}}}, Scale Curve=1.25)
    Index 0 Param 2=&FloatFromCurve(
     FloatCurve=DataInterface() {propertyValues {Curve {keys [<at 0 1 RCIM_Cubic>,<at 0.5 0.1 RCIM_Cubic>,<at 1 1 RCIM_Cubic>]}}}
     Scale Curve=3.5
    )
    Index 0 Param 4=1
    Module.Param0WriteEnabled=true
    Module.Param1WriteEnabled=true
    Module.Param2WriteEnabled=true
    Module.Param3WriteEnabled=true
   )
  ]
  rendererClasses [&NiagaraSpriteRendererProperties, &NiagaraMeshRendererProperties]
  renderers [
   {rendererClass=&NiagaraSpriteRendererProperties}
   {rendererClass=&NiagaraMeshRendererProperties, rendererIndex=1}
  ]
 }
 FN_AttachedRibbon_v3 {
  bEnabled=false
  emitterUpdateScript [
   &EmitterState(
    Enable Distance Culling=true
    Min Distance Response=::SleepAndLetParticlesFinish
    MinDistance=180
    Module.bUseMaxDistance=false
    Module.bUseMinDistance=true
    Scalability Mode=::NewEnumerator1
   )
   /Game/Effects/Niagara/Modules/Ribbons/FN_SpawnRibbon_v3(
    Number of Ribbons:NiagaraInt32=&Multiply_Int(A=6, B=&MakeCustomIntFromBool(Bool=$User.IsSkydiving, True INT=0, False INT=1))
    Skeleton Mesh:NiagaraDataInterfaceSkeletalMesh=$User.SkeletalMesh
    Spawn Mode:ENiagaraSkeletalSpawning=::NewEnumerator1
    SpawnRate:NiagaraFloat=&FloatFromCurve(
     FloatCurve=DataInterface() {propertyValues {Curve {keys [{arriveTangent=2.3455, interpMode=RCIM_Cubic, leaveTangent=2.3455, tangentMode=RCTM_User, time=0, value=0}, <at 0.5 1 RCIM_Cubic>]}}}
     CurveIndex=$System.FN_Integrator.Value
     Scale Curve=20
    )
   )
   &SpawnBurst_Instantaneous(Spawn Count=6, Spawn Group=-1, Use Spawn Probability=false)
  ]
  particleSpawnScript [
   /Game/Effects/Niagara/Modules/Ribbons/FN_InitializeRibbon_v4(
    Enable Skeletal Position=true
    Lifetime:NiagaraFloat=&Lerp_Float(
     A=1
     B=0.01
     Alpha=&Divide_Float(A=&VectorLength(Vector=$Engine.Owner.Velocity), B=2000)
    )
    Ribbon Width=10
    SkeletalMesh:NiagaraDataInterfaceSkeletalMesh=$User.SkeletalMesh
   )
   &AddVelocityFromPoint(Module.Use Falloff Distance=false, Origin Offset=<0 0 100>, Velocity Strength=25)
   &InheritVelocity(Module.Limit Speed=true, Velocity Limit=450, Velocity Scale=<0.7 0.7 0.7>)
  ]
  particleUpdateScript [
   &ParticleState()
   /SetVariables(Particles.DistanceFromBody:NiagaraFloat=&NormalizeFloat(
     Value=&VectorLength(Vector=&Subtract_Vector(A=Linked() {linkedVariable {name=Particles.Position, type=NiagaraPosition}}, B=$Particles.FN_UpdateRibbon_v4.SampledPosition))
     Min=0
     Max=50
    )) {isSetParameters=true, module=SetVariables_#3}
   /SetVariables(Particles.Speed:NiagaraFloat=&NormalizeFloat(Value=&VectorLength(Vector=$Particles.Velocity), Min=0, Max=450)) {isSetParameters=true, module=SetVariables_#4}
   /Niagara/Modules/Update/Forces/CurlNoiseForce(
    Noise Frequency:NiagaraFloat=&Lerp_Float(A=0.03, B=0.08, Alpha=$Particles.Speed)
    Noise Strength=&FloatFromCurve(FloatCurve=DataInterface() {propertyValues {Curve {keys [<at 0.1 25>,<at 1 2000>]}}}, CurveIndex=$Particles.Speed)
    Pan Noise=true
    Pan Noise Field=<0.05 0.05 0.05>
    Random Seed:NiagaraInt32=&RandomRangeInt(Minimum=0, Maximum=1)
   )
   &Drag(Drag=0.9, Use Linear Drag=true, Use Rotational Drag=false)
   &SolveForcesAndVelocity(Clamp Velocity=false, Limit Acceleration=false)
   /Game/Effects/Niagara/Modules/Ribbons/FN_UpdateRibbon_v4(Enable Skeletal Position=true, SkeletalMesh:NiagaraDataInterfaceSkeletalMesh=$User.SkeletalMesh)
   &ScaleRibbonWidth(Ribbon Width Scale=&Multiply_Float(
     A=&Multiply_Float(
      A=&FloatFromCurve(
       FloatCurve=DataInterface() {propertyValues {Curve {keys [{arriveTangent=19.0097, interpMode=RCIM_Cubic, leaveTangent=-1.3891, tangentMode=RCTM_User, time=0, value=1}, <at 1 0 RCIM_Cubic>]}}}
       Scale Curve=0.8
      )
      B=&FloatFromCurve(
       FloatCurve=DataInterface() {propertyValues {Curve {keys [{arriveTangent=-7.8567, time=0, value=0}, <at 0.4 1>]}}}
       CurveIndex=$System.FN_Integrator.Value
      )
     )
     B=&FloatFromCurve(
      FloatCurve=DataInterface() {propertyValues {Curve {keys [
          {arriveTangent=-0.3908, interpMode=RCIM_Cubic, time=0, value=1}
          {arriveTangent=-1, interpMode=RCIM_Cubic, leaveTangent=-1, time=0.7899, value=0.7951}
          {interpMode=RCIM_Cubic, leaveTangent=-2.5446, time=1, value=0}
         ]}}}
      CurveIndex=$Particles.DistanceFromBody
     )
    ))
   &CameraOffset(Camera Offset Amount=15)
   &DynamicMaterialParameters(
    Index 0 Param 1=&FloatFromCurve(FloatCurve=DataInterface() {propertyValues {Curve {keys [<at 0 1>,<at 1 0>]}}})
    Index 0 Param 2=$Particles.MaterialRandom
    Index 0 Param 3=&Lerp_Float(A=1, B=2, Alpha=$Particles.PlayerSpeed)
    Index 0 Param 4=1
    Module.Param0WriteEnabled=true
    Module.Param1WriteEnabled=true
    Module.Param2WriteEnabled=false
    Module.Param3WriteEnabled=true
   )
  ]
  rendererClasses [&NiagaraRibbonRendererProperties]
  renderers [{rendererClass=&NiagaraRibbonRendererProperties}]
 }
 FN_AttachedRibbon_v3001 {
  emitterUpdateScript [
   &EmitterState(
    Enable Distance Culling=true
    Inactive Response=::NewEnumerator2
    Life Cycle Mode=::NewEnumerator1
    Loop Duration=1
    Max Distance Response=::Awaken
    MaxDistance=201
    Min Distance Response=::SleepAndLetParticlesFinish
    MinDistance=200
    Module.bUseMaxDistance=true
    Module.bUseMinDistance=true
    Reset Age On Awaken=true
    Scalability Mode=::NewEnumerator1
    UseLoopDelay=false
   )
   &FN_SpawnRibbon_v3(
    Number of Ribbons=6
    Skeleton Mesh:NiagaraDataInterfaceSkeletalMesh=$User.SkeletalMesh
    Spawn Mode:ENiagaraSkeletalSpawning=::NewEnumerator1
    SpawnRate:NiagaraFloat=&MakeCustomFloatFromBool(
     Bool=$User.IsSkydiving
     True Float=0
     False Float=&FloatFromCurve(
      FloatCurve=DataInterface() {propertyValues {Curve {keys [{tangentMode=RCTM_User, time=0, value=0}, {tangentMode=RCTM_User, time=1, value=1}]}}}
      CurveIndex=$System.FN_Integrator.NormalizedValue
      Scale Curve=20
     )
    )
   )
   &SpawnBurst_Instantaneous(Spawn Count=6, Spawn Group=-1, Use Spawn Probability=false) {enabled=false}
  ]
  particleSpawnScript [
   &FN_InitializeRibbon_v4(
    Enable Skeletal Position=true
    Lifetime:NiagaraFloat=&Lerp_Float(
     A=1
     B=0.01
     Alpha=&Divide_Float(A=&VectorLength(Vector=$Engine.Owner.Velocity), B=2000)
    )
    Ribbon Width=10
    SkeletalMesh:NiagaraDataInterfaceSkeletalMesh=$User.SkeletalMesh
   )
   &AddVelocityFromPoint(Module.Use Falloff Distance=false, Origin Offset=<0 0 100>, Velocity Strength=25)
   &InheritVelocity(Module.Limit Speed=true, Velocity Limit=450, Velocity Scale=<0.7 0.7 0.7>)
  ]
  particleUpdateScript [
   &ParticleState()
   /SetVariables(Particles.Speed:NiagaraFloat=&NormalizeFloat(Value=&VectorLength(Vector=$Particles.Velocity), Min=0, Max=450)) {isSetParameters=true, module=SetVariables_#4}
   &CurlNoiseForce(
    Noise Frequency:NiagaraFloat=&Lerp_Float(A=0.03, B=0.08, Alpha=$Particles.Speed)
    Noise Strength=&FloatFromCurve(FloatCurve=DataInterface() {propertyValues {Curve {keys [<at 0.1 25>,<at 1 2000>]}}}, CurveIndex=$Particles.Speed)
    Pan Noise=true
    Pan Noise Field=<0.05 0.05 0.05>
    Random Seed:NiagaraInt32=&RandomRangeInt(Minimum=0, Maximum=1)
   )
   &Drag(Drag=0.9, Use Linear Drag=true, Use Rotational Drag=false)
   &SolveForcesAndVelocity(Clamp Velocity=false, Limit Acceleration=false)
   &FN_UpdateRibbon_v4(Enable Skeletal Position=true, SkeletalMesh:NiagaraDataInterfaceSkeletalMesh=$User.SkeletalMesh)
   /SetVariables(Particles.DistanceFromBody:NiagaraFloat=&NormalizeFloat(
     Value=&VectorLength(Vector=&Subtract_Vector(A=Linked() {linkedVariable {name=Particles.Position, type=NiagaraPosition}}, B=$Particles.FN_UpdateRibbon_v4.SampledPosition))
     Min=0
     Max=100
    )) {isSetParameters=true, module=SetVariables_#3}
   &ScaleRibbonWidth(Ribbon Width Scale=&Multiply_Float(
     A=&Multiply_Float(
      A=&FloatFromCurve(
       FloatCurve=DataInterface() {propertyValues {Curve {keys [{arriveTangent=19.0097, interpMode=RCIM_Cubic, leaveTangent=-1.3891, tangentMode=RCTM_User, time=0, value=1}, <at 1 0 RCIM_Cubic>]}}}
       Scale Curve=0.85
      )
      B=&FloatFromCurve(
       FloatCurve=DataInterface() {propertyValues {Curve {keys [{arriveTangent=-7.8567, time=0, value=0}, <at 0.4 1>]}}}
       CurveIndex=$System.FN_Integrator.Value
      )
     )
     B=&FloatFromCurve(
      FloatCurve=DataInterface() {propertyValues {Curve {keys [
          {arriveTangent=-0.3908, interpMode=RCIM_Cubic, time=0, value=1}
          {arriveTangent=-1, interpMode=RCIM_Cubic, leaveTangent=-1, time=0.7899, value=0.7951}
          {interpMode=RCIM_Cubic, leaveTangent=-2.5446, time=1, value=0}
         ]}}}
      CurveIndex=$Particles.DistanceFromBody
     )
    ))
   &CameraOffset(Camera Offset Amount=15)
   &DynamicMaterialParameters(
    Index 0 Param 1=&FloatFromCurve(FloatCurve=DataInterface() {propertyValues {Curve {keys [<at 0 1>,<at 1 0>]}}})
    Index 0 Param 2=$Particles.MaterialRandom
    Index 0 Param 3=&Lerp_Float(A=1, B=2, Alpha=$Particles.PlayerSpeed)
    Index 0 Param 4=1
    Module.Param0WriteEnabled=true
    Module.Param1WriteEnabled=true
    Module.Param2WriteEnabled=false
    Module.Param3WriteEnabled=true
   )
   &KillParticles(Kill Particles=&SetBoolByFloatComparison(A=$Engine.Owner.LODDistance, B=200, Comparison Type=::NewEnumerator6))
  ]
  rendererClasses [&NiagaraRibbonRendererProperties, &NiagaraSpriteRendererProperties]
  renderers [
   {rendererClass=&NiagaraRibbonRendererProperties}
   {rendererClass=&NiagaraSpriteRendererProperties, rendererIndex=1}
  ]
 }
 MeshTendrils {
  bEnabled=false
  emitterUpdateScript [
   &EmitterState()
   &SpawnRate(SpawnRate=1.5, Use Spawn Probability=false)
  ]
  particleSpawnScript [
   &InitializeParticle(
    Lifetime=1.5
    Mesh Scale=&Multiply_VectorByFloat(Vector=<1 0.8 1>, Float=&RandomRangeFloat(Minimum=0.8, Maximum=1.25))
    Module.Write Color=true
    Module.Write Lifetime=true
    Module.Write Mass=true
    Module.Write Material Random=true
    Module.Write Position=false
    Module.Write Scale=true
    Module.Write SpriteRotation=false
    Module.Write SpriteSize=true
    Position=&SimulationPosition()
    Sprite Size=<25 25>
   )
   &FN_SetSpawnGroup_v2(Number of Spawn Groups=6)
   /Niagara/Modules/Spawn/Orientation/V2/InitialMeshOrientation(
    Rotation=&RandomRangeVector(Minimum=<-0.14 -0.07 0>, Maximum=<0.14 0.07 0>)
    Use Orientation Vector=true
    Use Rotation Vector=true
   )
   &self:DivideParticles_02(HowManyRenderers=3) {module=DivideParticles_02}
  ]
  particleUpdateScript [
   &ParticleState()
   &ScaleMeshSize(Scale Factor=&VectorFromFloat(Value=&FloatFromCurve(FloatCurve=DataInterface() {propertyValues {Curve {keys [
          {arriveTangent=3.7291, interpMode=RCIM_Cubic, leaveTangent=3.7291, tangentMode=RCTM_User, time=0, value=0}
          <at 0.5 1 RCIM_Cubic>
          {arriveTangent=-2.7655, interpMode=RCIM_Cubic, leaveTangent=-2.7655, tangentMode=RCTM_User, time=1, value=0}
         ]}}}))) {enabled=false}
   &DynamicMaterialParameters(
    Index 0 Param 1=&FloatFromCurve(FloatCurve=DataInterface() {propertyValues {Curve {keys [<at 0 0 RCIM_Cubic>,<at 0.5 1 RCIM_Cubic>,<at 1 0 RCIM_Cubic>]}}}, Scale Curve=1.25)
    Index 0 Param 2=&FloatFromCurve(
     FloatCurve=DataInterface() {propertyValues {Curve {keys [
         {arriveTangent=1.5805, interpMode=RCIM_Cubic, leaveTangent=1.5805, tangentMode=RCTM_User, time=0.0025, value=0.3751}
         {arriveTangent=0.0577, interpMode=RCIM_Cubic, leaveTangent=0.0577, time=0.493, value=0.9224}
         {arriveTangent=-1.6823, interpMode=RCIM_Cubic, leaveTangent=-1.6823, tangentMode=RCTM_User, time=1.0037, value=0.4328}
        ]}}}
     Scale Curve=1.4
    )
    Index 0 Param 4=1
    Module.Param0WriteEnabled=false
    Module.Param1WriteEnabled=true
    Module.Param2WriteEnabled=true
    Module.Param3WriteEnabled=true
   )
   &SkeletalMeshLocation(
    Apply Sampled Orientation to Mesh Particles=true
    Bone / Socket Index=$Particles.SpawnGroup
    Inherit Velocity=false
    MaskSpawnGroup=false
    Mesh Sampling Type=::NewEnumerator5
    Sampled Position Offset=&MakeVector(X=&RandomRangeFloat(Minimum=0, Maximum=25), Y=0, Z=0)
    Sampled Position Offset Coordinate Space:ENiagaraMeshTransforms=::NewEnumerator3
    Skeletal Mesh=$User.SkeletalMesh
    Socket Sampling Mode=::NewEnumerator5
    Use Sampled Mesh Attribute as Ribbon ID=false
   )
   &ScaleMeshSize(Scale Factor=&VectorFromCurve(
     VectorCurve=DataInterface() {propertyValues {
       XCurve {keys [<at 0 0 RCIM_Cubic>,<at 0.5 1 RCIM_Cubic>,<at 1 0 RCIM_Cubic>]}
       YCurve {keys [<at 0 0 RCIM_Cubic>,<at 0.5 1 RCIM_Cubic>,<at 1 0 RCIM_Cubic>]}
       ZCurve {keys [
         {arriveTangent=2.4741, interpMode=RCIM_Cubic, leaveTangent=2.4741, tangentMode=RCTM_User, time=0, value=0}
         <at 0.5 1 RCIM_Cubic>
         {arriveTangent=-2.7109, interpMode=RCIM_Cubic, leaveTangent=-2.7109, tangentMode=RCTM_User, time=1, value=0}
        ]}
      }}
     Scale Curve=&VectorFromFloat(Value=0.75)
    )) {module=ScaleMeshSize001}
  ]
  rendererClasses [&NiagaraMeshRendererProperties]
  renderers [
   {rendererClass=&NiagaraMeshRendererProperties, rendererIndex=0}
   {rendererClass=&NiagaraMeshRendererProperties, rendererIndex=1}
   {rendererClass=&NiagaraMeshRendererProperties, rendererIndex=2}
  ]
 }
 MeshTendrils001 {
  emitterUpdateScript [
   &EmitterState()
   &SpawnRate(SpawnRate=1.5, Use Spawn Probability=false)
  ]
  particleSpawnScript [
   &InitializeParticle(
    Lifetime=1.5
    Mesh Scale=&Multiply_VectorByFloat(Vector=<1 0.8 1>, Float=&RandomRangeFloat(Minimum=0.8, Maximum=1.4))
    Module.Write Color=true
    Module.Write Lifetime=true
    Module.Write Mass=true
    Module.Write Material Random=true
    Module.Write Position=false
    Module.Write Scale=true
    Module.Write SpriteRotation=false
    Module.Write SpriteSize=true
    Position=&SimulationPosition()
    Sprite Size=<25 25>
   )
   &FN_SetSpawnGroup_v2(Number of Spawn Groups=6)
   &InitialMeshOrientation(
    Rotation=&RandomRangeVector(Minimum=<-0.14 -0.07 0>, Maximum=<0.14 0.07 0>)
    Use Orientation Vector=true
    Use Rotation Vector=true
   )
   &self:DivideParticles_03(HowManyRenderers=3) {module=DivideParticles_03}
  ]
  particleUpdateScript [
   &ParticleState()
   &ScaleMeshSize(Scale Factor=&VectorFromFloat(Value=&FloatFromCurve(FloatCurve=DataInterface() {propertyValues {Curve {keys [
          {arriveTangent=3.7291, interpMode=RCIM_Cubic, leaveTangent=3.7291, tangentMode=RCTM_User, time=0, value=0}
          <at 0.5 1 RCIM_Cubic>
          {arriveTangent=-2.7655, interpMode=RCIM_Cubic, leaveTangent=-2.7655, tangentMode=RCTM_User, time=1, value=0}
         ]}}}))) {enabled=false}
   &DynamicMaterialParameters(
    Index 0 Param 1=&FloatFromCurve(FloatCurve=DataInterface() {propertyValues {Curve {keys [<at 0 0 RCIM_Cubic>,<at 0.5 1 RCIM_Cubic>,<at 1 0 RCIM_Cubic>]}}}, Scale Curve=1.25)
    Index 0 Param 2=&FloatFromCurve(
     FloatCurve=DataInterface() {propertyValues {Curve {keys [<at 0 1 RCIM_Cubic>,<at 0.5 0.1 RCIM_Cubic>,<at 1 1 RCIM_Cubic>]}}}
     Scale Curve=3.5
    )
    Index 0 Param 3=1
    Module.Param0WriteEnabled=true
    Module.Param1WriteEnabled=true
    Module.Param2WriteEnabled=true
    Module.Param3WriteEnabled=true
   )
   &SkeletalMeshLocation(
    Apply Sampled Orientation to Mesh Particles=true
    Bone / Socket Index=$Particles.SpawnGroup
    Inherit Velocity=false
    MaskSpawnGroup=false
    Mesh Sampling Type=::NewEnumerator5
    Sampled Position Offset=&MakeVector(X=&RandomRangeFloat(Minimum=0, Maximum=15), Y=0, Z=0)
    Sampled Position Offset Coordinate Space:ENiagaraMeshTransforms=::NewEnumerator3
    Skeletal Mesh=$User.SkeletalMesh
    Socket Sampling Mode=::NewEnumerator5
    Use Sampled Mesh Attribute as Ribbon ID=false
   )
   &ScaleMeshSize(Scale Factor=&VectorFromCurve(
     VectorCurve=DataInterface() {propertyValues {
       XCurve {keys [<at 0 0 RCIM_Cubic>,<at 0.5 1 RCIM_Cubic>,<at 1 0 RCIM_Cubic>]}
       YCurve {keys [<at 0 0 RCIM_Cubic>,<at 0.5 1 RCIM_Cubic>,<at 1 0 RCIM_Cubic>]}
       ZCurve {keys [
         {arriveTangent=2.4741, interpMode=RCIM_Cubic, leaveTangent=2.4741, tangentMode=RCTM_User, time=0, value=0}
         <at 0.5 1 RCIM_Cubic>
         {arriveTangent=-2.7109, interpMode=RCIM_Cubic, leaveTangent=-2.7109, tangentMode=RCTM_User, time=1, value=0}
        ]}
      }}
     Scale Curve=&VectorFromFloat(Value=0.85)
    )) {module=ScaleMeshSize001}
  ]
  rendererClasses [&NiagaraMeshRendererProperties]
  renderers [
   {rendererClass=&NiagaraMeshRendererProperties, rendererIndex=0}
   {rendererClass=&NiagaraMeshRendererProperties, rendererIndex=1}
   {rendererClass=&NiagaraMeshRendererProperties, rendererIndex=2}
  ]
 }
 Ribbons_Bone_Flow {
  bEnabled=false
  emitterSpawnScript [
   &FN_PartitionAttributes(Number of Bone Sets=2, Particles Per Partition=8, Partitions=2)
   /SetVariables(Emitter.Seed:NiagaraInt32=&RandomInt(Randomness Mode:ENiagaraRandomnessMode=::NewEnumerator0, Random Seed:NiagaraInt32=0, Fixed Random Seed:NiagaraBool=false)) {isSetParameters=true, module=SetVariables_#2}
  ]
  emitterUpdateScript [
   &EmitterState(
    Enable Distance Culling=true
    Min Distance Response=::SleepAndLetParticlesFinish
    MinDistance=180
    Module.bUseMaxDistance=false
    Module.bUseMinDistance=true
    Scalability Mode=::NewEnumerator1
   )
   &SpawnBurst_Instantaneous(Spawn Count=$Emitter.FN_PartitionAttributes.Spawn Total, Use Spawn Probability=false)
  ]
  particleSpawnScript [
   &FN_Particle_Partition_v5(Enable Ribbon ID=true, Seed:NiagaraInt32=$Emitter.Seed, Shuffle Partition ID=true)
   &InitializeParticle(
    Lifetime=0.5
    Module.Write Color=true
    Module.Write Lifetime=false
    Module.Write Mass=true
    Module.Write Material Random=true
    Module.Write Position=false
    Module.Write Scale=false
    Module.Write SpriteRotation=false
    Module.Write SpriteSize=true
    Position=&SimulationPosition()
    Sprite Size=<5 5>
   )
   &InitializeRibbon(
    Color=&MakeLinearColorFromVectorAndFloat("Vector (RGB)"=&MakeVectorFromLinearColorRGB(Color=<rgba 1 1 1 1>), "Float (Alpha)"=1)
    Lifetime=2.5
    Module.Write Color=true
    Module.Write Lifetime=true
    Module.Write Mass=true
    Module.Write Material Random=true
    Module.Write Position=false
    Module.Write RibbonTwist=false
    Module.Write RibbonWidth=true
    Ribbon Width=6.5
   )
   &FN_Initialize_BoneFlow_Attributes_v2(
    Angle:NiagaraFloat=&Multiply_Float(
     A=$Particles.Partition.NormalizedID
     B=&UniformRangedFloat(Minimum=0, Maximum=0.9, Random Seed=$Particles.Partition.ID)
    )
    Radius:NiagaraFloat=&FloatFromCurve(
     FloatCurve=DataInterface() {propertyValues {Curve {keys [
         {arriveTangent=34.0719, interpMode=RCIM_Cubic, leaveTangent=34.0719, tangentMode=RCTM_User, time=0, value=0}
         {arriveTangent=24.6242, interpMode=RCIM_Cubic, leaveTangent=24.6242, tangentMode=RCTM_User, time=0.2044, value=5.99}
         {arriveTangent=-27.8659, interpMode=RCIM_Cubic, leaveTangent=-27.8661, tangentMode=RCTM_User, time=0.8266, value=5.6263}
         {arriveTangent=-44.2444, interpMode=RCIM_Cubic, leaveTangent=-44.2448, tangentMode=RCTM_User, time=1, value=0}
        ]}}}
     CurveIndex=$Particles.Partition.NormalizedID
     Scale Curve=1.2
    )
    T:NiagaraFloat=$Particles.Partition.NormalizedID
   )
   &Material() {enabled=false}
  ]
  particleUpdateScript [
   &ParticleState()
   &FN_BoneFlow(
    Angle:NiagaraFloat=&Add_Float(A=$Particles.Initial.FN_Initialize_BoneFlow_Attributes_v2.Angle, B=HlslExpression() {hlslExpression="Emitter.Age  * 0.2"})
    Height:NiagaraFloat=$Particles.Partition.NormalizedID
    Radius:NiagaraFloat=$Particles.Initial.FN_Initialize_BoneFlow_Attributes_v2.Radius
    SkeletalMesh:NiagaraDataInterfaceSkeletalMesh=$User.SkeletalMesh
    Use Sockets Instead=true
    Write Position Attribute=true
   )
   &ScaleRibbonWidth(Ribbon Width Scale=&Multiply_Float(
     A=&FloatFromCurve(
      FloatCurve=DataInterface() {propertyValues {Curve {keys [<at 0 0.7 RCIM_Cubic>,<at 0.2 1 RCIM_Cubic>,<at 0.5 0.7 RCIM_Cubic>]}}}
      Scale Curve=0.7
     )
     B=&FloatFromCurve(
      FloatCurve=DataInterface() {propertyValues {Curve {keys [<at 0 0.7 RCIM_Cubic>,<at 0.4995 1.2 RCIM_Cubic>,<at 1 0.7 RCIM_Cubic>]}}}
      CurveIndex=$Particles.Partition.NormalizedID
     )
    ))
   &CameraOffset(Camera Offset Amount=15)
   &DynamicMaterialParameters(
    Index 0 Param 1=&FloatFromCurve(FloatCurve=DataInterface() {propertyValues {Curve {keys [<at 0 0 RCIM_Cubic>,<at 0.5 1 RCIM_Cubic>,<at 1 0 RCIM_Cubic>]}}})
    Index 0 Param 3=1
    Index 0 Param 4=1
    Module.Param0WriteEnabled=true
    Module.Param1WriteEnabled=false
    Module.Param2WriteEnabled=true
    Module.Param3WriteEnabled=true
   )
   /SetVariables(Particles.SpriteFacing=HlslExpression() {hlslExpression="dot(Particles.SpriteFacing,  float3(0,0,1))"}) {enabled=false, isSetParameters=true, module=SetVariables_#5}
  ]
  rendererClasses [&NiagaraRibbonRendererProperties]
  renderers [
   {rendererClass=&NiagaraRibbonRendererProperties}
   {rendererClass=&NiagaraRibbonRendererProperties, rendererIndex=1}
  ]
 }
 Ribbons_Bone_Flow001 {
  emitterSpawnScript [
   &FN_PartitionAttributes(Number of Bone Sets=2, Particles Per Partition=8, Partitions=2)
   /SetVariables(Emitter.Seed:NiagaraInt32=&RandomInt(Randomness Mode:ENiagaraRandomnessMode=::NewEnumerator0, Random Seed:NiagaraInt32=0, Fixed Random Seed:NiagaraBool=false)) {isSetParameters=true, module=SetVariables_#2}
  ]
  emitterUpdateScript [
   &EmitterState(
    Enable Distance Culling=true
    Min Distance Response=::SleepAndLetParticlesFinish
    MinDistance=180
    Module.bUseMaxDistance=false
    Module.bUseMinDistance=true
    Scalability Mode=::NewEnumerator1
   )
   &SpawnBurst_Instantaneous(Spawn Count=$Emitter.FN_PartitionAttributes.Spawn Total, Use Spawn Probability=false)
  ]
  particleSpawnScript [
   &FN_Particle_Partition_v5(Enable Ribbon ID=true, Seed:NiagaraInt32=$Emitter.Seed, Shuffle Partition ID=true)
   &InitializeParticle(
    Lifetime=0.5
    Module.Write Color=true
    Module.Write Lifetime=false
    Module.Write Mass=true
    Module.Write Material Random=true
    Module.Write Position=false
    Module.Write Scale=false
    Module.Write SpriteRotation=false
    Module.Write SpriteSize=true
    Position=&SimulationPosition()
    Sprite Size=<5 5>
   )
   &InitializeRibbon(
    Color=&MakeLinearColorFromVectorAndFloat("Vector (RGB)"=&MakeVectorFromLinearColorRGB(Color=<rgba 1 1 1 1>), "Float (Alpha)"=1)
    Lifetime=2.5
    Module.Write Color=true
    Module.Write Lifetime=true
    Module.Write Mass=true
    Module.Write Material Random=true
    Module.Write Position=false
    Module.Write RibbonTwist=false
    Module.Write RibbonWidth=true
    Ribbon Width=6.5
   )
   &FN_Initialize_BoneFlow_Attributes_v2(
    Angle:NiagaraFloat=&Multiply_Float(
     A=$Particles.Partition.NormalizedID
     B=&UniformRangedFloat(Minimum=0, Maximum=0.9, Random Seed=$Particles.Partition.ID)
    )
    Radius:NiagaraFloat=&FloatFromCurve(
     FloatCurve=DataInterface() {propertyValues {Curve {keys [
         {arriveTangent=34.0719, interpMode=RCIM_Cubic, leaveTangent=34.0719, tangentMode=RCTM_User, time=0, value=0}
         {arriveTangent=24.6242, interpMode=RCIM_Cubic, leaveTangent=24.6242, tangentMode=RCTM_User, time=0.2044, value=5.99}
         {arriveTangent=-27.8659, interpMode=RCIM_Cubic, leaveTangent=-27.8661, tangentMode=RCTM_User, time=0.8266, value=5.6263}
         {arriveTangent=-44.2444, interpMode=RCIM_Cubic, leaveTangent=-44.2448, tangentMode=RCTM_User, time=1, value=0}
        ]}}}
     CurveIndex=$Particles.Partition.NormalizedID
     Scale Curve=1.2
    )
    T:NiagaraFloat=$Particles.Partition.NormalizedID
   )
   &Material() {enabled=false}
  ]
  particleUpdateScript [
   &ParticleState()
   &FN_BoneFlow(
    Angle:NiagaraFloat=&Add_Float(A=$Particles.Initial.FN_Initialize_BoneFlow_Attributes_v2.Angle, B=HlslExpression() {hlslExpression="Emitter.Age  * 0.2"})
    Height:NiagaraFloat=$Particles.Partition.NormalizedID
    Radius:NiagaraFloat=$Particles.Initial.FN_Initialize_BoneFlow_Attributes_v2.Radius
    SkeletalMesh:NiagaraDataInterfaceSkeletalMesh=$User.SkeletalMesh
    Use Sockets Instead=true
    Write Position Attribute=true
   )
   &ScaleRibbonWidth(Ribbon Width Scale=&Multiply_Float(
     A=&FloatFromCurve(
      FloatCurve=DataInterface() {propertyValues {Curve {keys [<at 0 0.7 RCIM_Cubic>,<at 0.2 1 RCIM_Cubic>,<at 0.5 0.7 RCIM_Cubic>]}}}
      Scale Curve=0.7
     )
     B=&FloatFromCurve(
      FloatCurve=DataInterface() {propertyValues {Curve {keys [<at 0 0.7 RCIM_Cubic>,<at 0.4995 1.2 RCIM_Cubic>,<at 1 0.7 RCIM_Cubic>]}}}
      CurveIndex=$Particles.Partition.NormalizedID
     )
    ))
   &CameraOffset(Camera Offset Amount=15)
   &DynamicMaterialParameters(
    Index 0 Param 1=&FloatFromCurve(FloatCurve=DataInterface() {propertyValues {Curve {keys [<at 0 0 RCIM_Cubic>,<at 0.5 1 RCIM_Cubic>,<at 1 0 RCIM_Cubic>]}}})
    Index 0 Param 3=1
    Index 0 Param 4=1
    Module.Param0WriteEnabled=true
    Module.Param1WriteEnabled=false
    Module.Param2WriteEnabled=true
    Module.Param3WriteEnabled=true
   )
   /SetVariables(Particles.SpriteFacing=HlslExpression() {hlslExpression="dot(Particles.SpriteFacing,  float3(0,0,1))"}) {enabled=false, isSetParameters=true, module=SetVariables_#5}
  ]
  rendererClasses [&NiagaraRibbonRendererProperties]
  renderers [{rendererClass=&NiagaraRibbonRendererProperties}]
 }
}
summary {
 systemName=NS_SKin_Vertigo
 emitters [
  {
   bEnabled=false
   emitterName=Circles_Bone_Flow
   rendererClasses [&NiagaraSpriteRendererProperties, &NiagaraMeshRendererProperties]
  }
  {bEnabled=false, emitterName=MeshTendrils, rendererClasses [&NiagaraMeshRendererProperties]}
  {bEnabled=false, emitterName=Ribbons_Bone_Flow, rendererClasses [&NiagaraRibbonRendererProperties]}
  {bEnabled=false, emitterName=FN_AttachedRibbon_v3, rendererClasses [&NiagaraRibbonRendererProperties]}
  {
   bEnabled=false
   emitterName=Circles_Bone_Flow001
   rendererClasses [&NiagaraSpriteRendererProperties, &NiagaraMeshRendererProperties]
  }
  {
   emitterName=Circles_Bone_Flow002
   rendererClasses [&NiagaraSpriteRendererProperties, &NiagaraMeshRendererProperties]
  }
  {emitterName=MeshTendrils001, rendererClasses [&NiagaraMeshRendererProperties]}
  {emitterName=Ribbons_Bone_Flow001, rendererClasses [&NiagaraRibbonRendererProperties]}
  {
   emitterName=FN_AttachedRibbon_v3001
   rendererClasses [&NiagaraRibbonRendererProperties, &NiagaraSpriteRendererProperties]
  }
 ]
 userVariables [
  User.Color_Tint {type=LinearColor, description="", defaultValue {struct=LinearColor, value=<rgba 1 1 1 1>}}
  User.EmissiveIntensity {type=NiagaraFloat, defaultValue=0, description=""}
  User.IsSkydiving {type=NiagaraBool}
  User.RibbonIDOffset {type=NiagaraInt32}
  User.SkeletalMesh {
   type=NiagaraDataInterfaceSkeletalMesh
   defaultValue {
    struct=NiagaraExt_VariableValue_DataInterface
    value {dataInterface=&self:NiagaraDataInterfaceSkeletalMesh_0}
   }
  }
 ]
}
